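; Blueprint for the female Renegade Skull pickaxe cosmetic in Battle Royale, adding a glowing shell over the weapon.
; Its construction script runs the parent logic, then assigns the glow material to slot 0 of the RenegadeSkullShellMesh skinned mesh component.
; It also forces that mesh to a fixed LOD.
; The Tick event is empty.
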(event EventTick
  (params (DeltaSeconds :type Float%20%28single-precision%29)))

(fn ConstructionScript
  (bind _renegadeskullshellmesh (Variables|Default|GetRenegadeSkullShellMesh))
  (|CallParentFunction)
  (Rendering|Material|SetMaterial _renegadeskullshellmesh 0 "/BRCosmetics/Weapons/FORT_Melee/Pickaxe_Renegade_Skull_Female/FX/M_Renegade_Skull_Pickaxe_Glow.M_Renegade_Skull_Pickaxe_Glow")
  (Components|SkinnedMesh|SetForcedLOD _renegadeskullshellmesh 2))
