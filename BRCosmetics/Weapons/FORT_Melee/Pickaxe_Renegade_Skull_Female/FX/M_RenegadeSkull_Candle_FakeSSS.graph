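# Unlit alpha-composite material that fakes subsurface scattering glow on the Renegade Skull pickaxe candle, set up for Niagara mesh particles.
# Boosted vertex color minus an animated threshold masks the glow, which a pixel-normal Fresnel then shapes.
# A zero-tiled panning texture offsets time in a sine, making the threshold flicker uniformly across the mesh.
# Particle color tints the mask through an emissive-adjustment function, and particle alpha scales it.
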Material /BRCosmetics/Weapons/FORT_Melee/Pickaxe_Renegade_Skull_Female/FX/M_RenegadeSkull_Candle_FakeSSS
graph {
 properties {
  bUsedWithNiagaraMeshParticles=true
  blendMode=BLEND_AlphaComposite
  pixelDepthOffsetMode=PDOM_Legacy
  previewMesh=&/SM_Renegade_Skull_Candle
  shadingModel=MSM_Unlit
  usedShadingModels=Unlit
 }
 nodes [
  n1 = VertexColor()
  n2 = Multiply(A=@n1) {constB=1.5}
  n3 = Time()
  n4 = TextureSample(UVs=Panner(Coordinate=TextureCoordinate() {uTiling=0, vTiling=0}) {speedY=0.5}) {texture=&2ndInfectedMapLinear}
  n5 = LinearInterpolate(Alpha=Sine(None=Add(A=@n3, B=@n4.R)) {period=0.5}) {constA=0.4, constB=0.2}
  n6 = Multiply(
   A=Saturate(None=Subtract(A=@n2, B=@n5))
   B=MaterialFunctionCall(Exponent=Constant() {r=0.5}, Vertex / Pixel Normal=PixelNormalWS()) {materialFunction=&MF_Fresnel_FX}.Fresnel
  )
  n7 = ParticleColor()
  MaterialFunctionCall(Emissive=Multiply(A=@n6, B=@n7.RGB)) {materialFunction=&MF_FX_EmissiveAdjustments}
  Multiply(A=@n6, B=@n7.A)
  Power(Base=@n1)
 ]
}
