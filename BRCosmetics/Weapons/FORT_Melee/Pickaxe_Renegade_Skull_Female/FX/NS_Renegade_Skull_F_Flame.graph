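# Niagara system for the blue candle flame on the female Renegade Skull pickaxe, with all four emitters following sampled locations on the User.SkeletalMesh.
# Flame and Smoke spawn counter-ordered ribbons, pinned briefly at birth, then shaped by upward gravity, curl noise and velocity-scaled drag.
# Flame culls particles straying from the sampled point, while Smoke erodes with age through ribbon material parameters.
# CandleSSS holds a mesh; Glow adds a flickering sprite scaled by owner speed.
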NiagaraSystem /BRCosmetics/Weapons/FORT_Melee/Pickaxe_Renegade_Skull_Female/FX/NS_Renegade_Skull_F_Flame
dependencies {
 usedDataInterfaces [/Script/Niagara.NiagaraDataInterfaceSimpleCounter, &NiagaraDataInterfaceCurve, &NiagaraDataInterfaceColorCurve]
 usedDynamicInputs []
 usedModules [&SystemState]
}
emitters {
 CandleSSS {
  emitterUpdateScript [
   &EmitterState(Life Cycle Mode=::NewEnumerator1, Loop Behavior=::NewEnumerator1, Loop Duration Mode=::NewEnumerator1, UseLoopDelay=false)
   &SpawnBurst_Instantaneous(Use Spawn Probability=false, UseLoopCountLimit=false)
  ]
  particleSpawnScript [&InitializeParticle(Color=<rgba 0 1.1145 2 1>, UsePositionOffset=false, Write Material Random=true)]
  particleUpdateScript [
   &ParticleState(
    Kill Particles When Lifetime Has Elapsed=false
    Let Infinitely Lived Particles Die When Emitter Deactivates=true
    Loop Particles Lifetime=true
   )
   &SkeletalMeshLocation(
    Apply Sampled Orientation to Mesh Particles=true
    Inherit Velocity=false
    "Inherit Velocity (Scale)"=0
    Interpolate Mesh Sampling=false
    MaskSpawnGroup=false
    Mesh Sampling Type=::NewEnumerator5
    Sampled Position Offset=<0 0 -7.65>
    Sampled Position Offset Coordinate Space:ENiagaraMeshTransforms=::NewEnumerator2
    Skeletal Mesh=$User.SkeletalMesh
    Socket Sampling Mode=::NewEnumerator5
   ) {module=SkeletalMeshLocation001}
  ]
  rendererClasses [&NiagaraMeshRendererProperties]
  renderers [{rendererClass=&NiagaraMeshRendererProperties, rendererIndex=0}]
 }
 Flame {
  emitterSpawnScript [/SetVariables(Emitter.Depthfade=2, Emitter.Hardness=0.6, Emitter.Power=2) {isSetParameters=true}]
  emitterUpdateScript [
   &EmitterState()
   &SpawnRate(SpawnRate=50, Use Spawn Probability=false)
  ]
  particleSpawnScript [
   &InitializeRibbon(
    Color=<rgba 0 1.0132 5 0.3>
    Lifetime=&RandomRangeFloat(Minimum=0.1, Maximum=0.5)
    Module.Write Color=true
    Module.Write Lifetime=true
    Module.Write Mass=true
    Module.Write Material Random=true
    Module.Write Position=true
    Module.Write RibbonTwist=false
    Module.Write RibbonWidth=true
    Ribbon Width=&RandomRangeFloat(Minimum=3, Maximum=5)
   )
   &SkeletalMeshLocation(
    Inherit Velocity=true
    MaskSpawnGroup=false
    Mesh Sampling Type=::NewEnumerator5
    Sampled Position Offset Coordinate Space:ENiagaraMeshTransforms=::NewEnumerator3
    Skeletal Mesh=$User.SkeletalMesh
    Socket Sampling Mode=::NewEnumerator5
    VelocitySpeedLimit=false
   )
   /SetVariables(Particles.RibbonLinkOrder:NiagaraFloat=&MakeFloatFromInt(Int=/Niagara/DynamicInputs/DataInterface/GenerateIntFromCounter(SimpleCounter:NiagaraDataInterfaceSimpleCounter=DataInterface() {propertyValues {GpuSyncMode=None, InitialValue=0}}))) {isSetParameters=true, module=SetVariables_#1}
  ]
  particleUpdateScript [
   &ParticleState()
   &SkeletalMeshLocation(
    Apply To Particle Position:NiagaraBool=&SetBoolByFloatComparison(A=$Particles.NormalizedAge, B=0.1, Comparison Type=::NewEnumerator6)
    Inherit Velocity=false
    Interpolate Mesh Sampling=false
    MaskSpawnGroup=false
    Mesh Sampling Type=::NewEnumerator5
    Sampled Position Offset Coordinate Space:ENiagaraMeshTransforms=::NewEnumerator3
    Skeletal Mesh=$User.SkeletalMesh
    Socket Sampling Mode=::NewEnumerator5
   ) {module=SkeletalMeshLocation001}
   &GravityForce(Gravity=<0 0 500>)
   /Niagara/Modules/Update/Forces/CurlNoiseForce(
    Noise Strength=125
    Pan Noise=true
    Pan Noise Field=<0 0 1>
    Random Seed:NiagaraInt32=&RandomRangeInt(Minimum=0, Maximum=100)
   )
   &Drag(
    Drag=&Multiply_Float(A=/Niagara/DynamicInputs/TypeConversions/VectorLengthSafe(Vector:Vector3f=$Output.SkeletalMeshLocation001.SampledVelocity), B=0.9)
    Use Linear Drag=true
    Use Rotational Drag=false
   )
   &InheritSourceMovement(
    Applied Position Falloff Curve=DataInterface() {propertyValues {Curve {keys [
        {arriveTangent=3.6987, interpMode=RCIM_Cubic, leaveTangent=3.6987, tangentMode=RCTM_User, time=0, value=0.1}
        {arriveTangent=0.3482, interpMode=RCIM_Cubic, leaveTangent=0.3482, tangentMode=RCTM_User, time=0.3, value=0.95}
        {arriveTangent=-4.3675, interpMode=RCIM_Cubic, leaveTangent=-4.3675, tangentMode=RCTM_User, time=1, value=0}
       ]}}}
    Applied Position Falloff Distance=90
    ApplyForce=false
    ApplyPosition=true
    ApplyPositionFalloff=true
    ApplyPositionFalloffCurve=false
   )
   &SolveForcesAndVelocity(Clamp Velocity=true, Limit Acceleration=false)
   &ScaleColor(
    Scale Alpha=&FloatFromCurve(FloatCurve=DataInterface() {propertyValues {Curve {keys [<at 0 1 RCIM_Cubic>, {arriveTangent=-1, interpMode=RCIM_Cubic, leaveTangent=-1, time=0.2, value=1}, <at 1 0 RCIM_Cubic>]}}})
    ScaleA=true
    ScaleRGB=true
   )
   /SetVariables(
    Particles.DistanceScale:NiagaraFloat=&Add_Float(
     A=1
     B=&ClampFloat(
      Float=&Multiply_Float(
       A=-0.1
       B=&VectorLength(Vector=&Subtract_Vector(
         A=Linked() {linkedVariable {name=Particles.Position, type=NiagaraPosition}}
         B=Linked() {linkedVariable {name=Particles.Initial.Position, type=NiagaraPosition}}
        ))
      )
      Min=-0.6
      Max=0
     )
    )
    Particles.VelocityScale:NiagaraFloat=&Add_Float(
     A=10
     B=&Multiply_Float(A=-0.005, B=&VectorLengthSafe(Vector:Vector3f=$Output.SkeletalMeshLocation001.SampledVelocity))
    )
   ) {isSetParameters=true}
   &KillParticlesInVolume(
    Invert Volume=true
    Origin Offset:Vector3f=&Multiply_VectorByFloat(Vector=$Output.SkeletalMeshLocation001.SampledVelocity, Float=0.01)
    Sphere Radius=15
    Volume Origin:Vector3f=$Output.SkeletalMeshLocation001.SampledPosition
   ) {module=KillParticlesInVolume002}
   &ScaleRibbonWidth(Ribbon Width Scale=&FloatFromCurve(
     FloatCurve=DataInterface() {propertyValues {Curve {keys [
         {arriveTangent=8.9571, interpMode=RCIM_Cubic, leaveTangent=8.9571, tangentMode=RCTM_User, time=0, value=0.875}
         {arriveTangent=-0.0621, interpMode=RCIM_Cubic, leaveTangent=-0.0621, tangentMode=RCTM_User, time=0.2036, value=1.9766}
         {arriveTangent=0.0338, interpMode=RCIM_Cubic, leaveTangent=0.0338, tangentMode=RCTM_User, time=1, value=0}
        ]}}}
     Scale Curve=&MakeCustomFloatFromBool(
      Bool=&SetBoolByFloatComparison(A=$Particles.DistanceScale, B=0.5)
      True Float=1
      False Float=$Particles.DistanceScale
     )
    ))
   &DynamicMaterialParameters(
    Index 0 Param 1=5
    Module.Param0WriteEnabled=true
    Module.Param1WriteEnabled=false
    Module.Param2WriteEnabled=false
    Module.Param3WriteEnabled=false
   )
  ]
  rendererClasses [&NiagaraRibbonRendererProperties]
  renderers [{rendererClass=&NiagaraRibbonRendererProperties}]
 }
 Glow {
  emitterSpawnScript [/SetVariables(Emitter.Depthfade=10, Emitter.Hardness=0.15, Emitter.Power=1.25) {isSetParameters=true}]
  emitterUpdateScript [
   &EmitterState(Life Cycle Mode=::NewEnumerator1, Loop Behavior=::NewEnumerator1, Loop Duration Mode=::NewEnumerator1, UseLoopDelay=false)
   &SpawnBurst_Instantaneous(Use Spawn Probability=false, UseLoopCountLimit=false)
   /SetVariables(Emitter.Power:NiagaraFloat=&Add_Float(
     A=1.25
     B=&Waveform(
      Waveform Count=::NewEnumerator1
      "Waveform [1]"=::NewEnumerator0
      "[1] Phase Scale"=0.2215
      "[1] Frequency"=0.451
      "[1] Amplitude Min/Max"=<0 1>
      "[1] Scale"=0.95
      "[2] Average"=false
      "[2] Input Position"=0
      "[2] Phase"=$Emitter.Age
      "[2] Phase Scale"=0.234
      "[2] Compound Frequency"=2
      "[2] Frequency"=0.521
      "[2] Amplitude Min/Max"=<0 1>
      "[2] Scale"=1
      "[2] Bias"=0
      "[2] Exponent"=1
      "[2] Pulse Duration"=0.1
      "[2] Clip Triangle Min/Max"=<0 0>
      "[2] Random Seed"=0
      "[3] Phase"=$Emitter.Age
     )
    )) {isSetParameters=true}
  ]
  particleSpawnScript [
   &InitializeParticle(Color=<rgba 0 0.526 1 0.25>, Sprite Size Mode=::NewEnumerator3, Uniform Sprite Size=65, UsePositionOffset=false, Write Material Random=true)
   &CameraOffset(Camera Offset Amount=5)
  ]
  particleUpdateScript [
   &ParticleState(
    Kill Particles When Lifetime Has Elapsed=false
    Let Infinitely Lived Particles Die When Emitter Deactivates=true
    Loop Particles Lifetime=true
   )
   &SkeletalMeshLocation(
    Apply Sampled Orientation to Mesh Particles=true
    Inherit Velocity=false
    "Inherit Velocity (Scale)"=0
    Interpolate Mesh Sampling=false
    MaskSpawnGroup=false
    Mesh Sampling Type=::NewEnumerator5
    Sampled Position Offset=<0 0 1>
    Sampled Position Offset Coordinate Space:ENiagaraMeshTransforms=::NewEnumerator1
    Skeletal Mesh=$User.SkeletalMesh
    Socket Sampling Mode=::NewEnumerator5
   ) {module=SkeletalMeshLocation001}
   &ScaleColor(
    Scale RGB=&Multiply_VectorByFloat(
     Vector=<1 1 1>
     Float=&FloatFromCurve(
      FloatCurve=DataInterface() {propertyValues {Curve {keys [<at 0 0.8>,<at 0.1 0.9>,<at 0.3 0.7>,<at 0.4 0.8>,<at 0.5 0.8>,<at 0.6 1>,<at 0.8 0.75>,<at 0.9 0.95>,<at 1 0.8>]}}}
      Scale Curve=&Multiply_Float(A=1, B=1)
     )
    )
    ScaleA=true
    ScaleRGB=true
   )
   /SetVariables() {isSetParameters=true}
   /Niagara/Modules/Update/Color/ScaleColorBySpeed(Max Speed Threshold=500, Min Speed Threshold=0, RGB Scale Min=0.1, Source Velocity:Vector3f=$Engine.Owner.Velocity)
  ]
  rendererClasses [&NiagaraSpriteRendererProperties]
  renderers [{rendererClass=&NiagaraSpriteRendererProperties}]
 }
 Smoke {
  emitterUpdateScript [
   &EmitterState()
   &SpawnRate(SpawnRate=50, Use Spawn Probability=false)
  ]
  particleSpawnScript [
   &InitializeRibbon(
    Color=<rgba 0 7.6606 10 0.3>
    Lifetime=0.5
    Module.Write Color=true
    Module.Write Lifetime=true
    Module.Write Mass=true
    Module.Write Material Random=true
    Module.Write Position=true
    Module.Write RibbonTwist=false
    Module.Write RibbonWidth=true
    Ribbon Width=&RandomRangeFloat(Minimum=1, Maximum=6)
   )
   /SetVariables(Particles.SpriteSize=&Vector2DFromFloat(Value=1)) {isSetParameters=true}
   &SkeletalMeshLocation(
    Inherit Velocity=true
    MaskSpawnGroup=false
    Mesh Sampling Type=::NewEnumerator5
    Sampled Position Offset Coordinate Space:ENiagaraMeshTransforms=::NewEnumerator3
    Skeletal Mesh=$User.SkeletalMesh
    Socket Sampling Mode=::NewEnumerator5
    VelocitySpeedLimit=false
   )
   /SetVariables(Particles.RibbonLinkOrder:NiagaraFloat=&MakeFloatFromInt(Int=&GenerateIntFromCounter(SimpleCounter:NiagaraDataInterfaceSimpleCounter=DataInterface() {propertyValues {GpuSyncMode=None, InitialValue=0}}))) {isSetParameters=true, module=SetVariables_#1}
  ]
  particleUpdateScript [
   &ParticleState()
   &SkeletalMeshLocation(
    Apply To Particle Position:NiagaraBool=&SetBoolByFloatComparison(A=$Particles.NormalizedAge, B=0.2, Comparison Type=::NewEnumerator6)
    Inherit Velocity=false
    Interpolate Mesh Sampling=false
    MaskSpawnGroup=false
    Mesh Sampling Type=::NewEnumerator5
    Sampled Position Offset Coordinate Space:ENiagaraMeshTransforms=::NewEnumerator3
    Skeletal Mesh=$User.SkeletalMesh
    Socket Sampling Mode=::NewEnumerator5
   ) {module=SkeletalMeshLocation001}
   /SetVariables(
    Particles.DistanceAbove:Vector3f=&self:ClampVector(
     Input Vector:Vector3f=&Multiply_Vector(
      A=<0 0 0.1>
      B=&Subtract_Vector(A=Linked() {linkedVariable {name=Particles.Position, type=NiagaraPosition}}, B=$Output.SkeletalMeshLocation001.SampledPosition)
     )
     Max:Vector3f=<0 0 2>
     Min:Vector3f=<0 0 0>
    )
    Particles.DistanceBelow:Vector3f=&self:ClampVector(
     Input Vector:Vector3f=&Multiply_Vector(
      A=<0 0 0.1>
      B=&Subtract_Vector(A=$Output.SkeletalMeshLocation001.SampledPosition, B=Linked() {linkedVariable {name=Particles.Position, type=NiagaraPosition}})
     )
     Max:Vector3f=<0 0 1>
     Min:Vector3f=<0 0 0>
    )
    Particles.DistanceFromBone:NiagaraFloat=&Multiply_Float(
     A=0.01
     B=&VectorLength(Vector=&Subtract_Vector(A=Linked() {linkedVariable {name=Particles.Position, type=NiagaraPosition}}, B=$Output.SkeletalMeshLocation001.SampledPosition))
    )
   ) {isSetParameters=true}
   /SetVariables(Particles.NormalizedAge:NiagaraFloat=&Lerp_Float(
     A=$Particles.NormalizedAge
     B=1
     Alpha=&ClampFloat(Float=&Add_Float(A=&MakeFloatFromVector(Vector=$Particles.DistanceAbove, Channel=::NewEnumerator2), B=-1.5))
    )) {isSetParameters=true}
   &GravityForce(Gravity=&Lerp_Vector(
     A=<0 0 1000>
     B=<0 0 400>
     Alpha=&ClampFloat(Float=&Multiply_Float(A=&VectorLength(Vector=$Output.SkeletalMeshLocation001.SampledVelocity), B=0.5))
    ))
   &CurlNoiseForce(
    Noise Frequency=5
    Noise Strength=&FloatFromCurve(FloatCurve=DataInterface() {propertyValues {Curve {keys [{arriveTangent=957.2371, interpMode=RCIM_Cubic, leaveTangent=957.237, tangentMode=RCTM_User, time=0, value=10}, <at 1 300>]}}})
    Pan Noise=true
    Pan Noise Field=<0 0 2>
    Random Seed:NiagaraInt32=&RandomRangeInt(Minimum=0, Maximum=100)
   )
   &Drag(
    Drag=&Multiply_Float(
     A=&ClampFloat(Float=&VectorLengthSafe(Vector:Vector3f=$Particles.Velocity), Max=5000)
     B=&Lerp_Float(A=0.5, B=0, Alpha=$Particles.NormalizedAge)
    )
    Use Linear Drag=true
    Use Rotational Drag=false
   )
   &AddVelocity(Velocity=&Multiply_VectorByFloat(Vector=$Particles.DistanceBelow, Float=&RandomRangeFloat(Minimum=0, Maximum=100))) {module=AddVelocity003}
   &AddVelocity(
    Coordinate Space=::World
    Velocity=&Lerp_Vector(
     A=<0 0 0>
     B=<0 0 100>
     Alpha=&FloatFromCurve(
      FloatCurve=DataInterface() {propertyValues {Curve {keys [{interpMode=RCIM_Cubic, tangentMode=RCTM_User, time=0.2, value=0}, <at 1 1>]}}}
      CurveIndex=$Particles.DistanceFromBone
      Scale Curve=&RandomRangeFloat(Minimum=0.25, Maximum=1)
     )
    )
   ) {enabled=false}
   /Game/Effects/Niagara/Modules/Materials/FN_RibbonMaterial/FN_RibbonMaterial(
    Color:LinearColor=&ColorFromCurve(ColorCurve=DataInterface() {propertyValues {
       AlphaCurve {keys [<at 0 1 RCIM_Cubic>]}
       BlueCurve {keys [<at 0 10 RCIM_Cubic>,<at 0.8703 0.918 RCIM_Cubic>]}
       GreenCurve {keys [<at 0 8.6989 RCIM_Cubic>,<at 0.8703 0.8015 RCIM_Cubic>]}
       RedCurve {keys [<at 0 0 RCIM_Cubic>,<at 0.8703 0.722 RCIM_Cubic>]}
      }})
    Color Pan U Speed=-2
    Color Scale U=7.5
    Erosion Amount:NiagaraFloat=&Lerp_Float(A=0.5, B=1.6, Alpha=$Particles.NormalizedAge)
    Erosion Pan Speed U=-0.8346
    Erossion Scale U=4.5
    Macro Distortion Amount=1
    Opacity Pan U Speed=-1.6786
    Opacity Power:NiagaraFloat=&RandomRangeFloat(Minimum=0.3, Maximum=1)
   ) {module=FN_RibbonMaterial002}
   &ScaleRibbonWidth(Ribbon Width Scale=&FloatFromCurve(
     FloatCurve=DataInterface() {propertyValues {Curve {keys [
         {arriveTangent=4.0699, interpMode=RCIM_Cubic, leaveTangent=4.0699, tangentMode=RCTM_User, time=0, value=0}
         <at 0.5 0.9666 RCIM_Cubic>
         {arriveTangent=-2.5115, interpMode=RCIM_Cubic, leaveTangent=-2.5115, tangentMode=RCTM_User, time=1, value=0}
        ]}}}
     Scale Curve=&RandomRangeFloat(Minimum=0.5, Maximum=2)
    )) {module=ScaleRibbonWidth001}
   &SolveForcesAndVelocity(Clamp Velocity=false, Limit Acceleration=false)
   &Color(Color=&ColorFromCurve(
     ColorCurve=DataInterface() {propertyValues {
       AlphaCurve {keys [<at 0 1 RCIM_Cubic>]}
       BlueCurve {keys [<at 0 0 RCIM_Cubic>,<at 0.4 0>,<at 0.6 0>,<at 1 1 RCIM_Cubic>]}
       GreenCurve {keys [<at 0 0.0148 RCIM_Cubic>,<at 0.4 0.8005>,<at 0.6 1>,<at 1 0.087 RCIM_Cubic>]}
       RedCurve {keys [<at 0 1 RCIM_Cubic>,<at 0.4 1>,<at 0.6 0.0681>,<at 1 0 RCIM_Cubic>]}
      }}
     CurveIndex=$Particles.DistanceFromBone
    )) {enabled=false}
  ]
  rendererClasses [&NiagaraSpriteRendererProperties, &NiagaraRibbonRendererProperties]
  renderers [
   {rendererClass=&NiagaraSpriteRendererProperties}
   {rendererClass=&NiagaraRibbonRendererProperties, rendererIndex=1}
  ]
 }
}
summary {
 systemName=NS_Renegade_Skull_F_Flame
 emitters [
  {emitterName=Smoke, rendererClasses [&NiagaraSpriteRendererProperties, &NiagaraRibbonRendererProperties]}
  {emitterName=CandleSSS, rendererClasses [&NiagaraMeshRendererProperties]}
  {emitterName=Glow, rendererClasses [&NiagaraSpriteRendererProperties]}
  {emitterName=Flame, rendererClasses [&NiagaraRibbonRendererProperties]}
 ]
 userVariables [
  User.bIsFrontEndPreview {type=NiagaraBool}
  User.SkeletalMesh {
   type=NiagaraDataInterfaceSkeletalMesh
   defaultValue {
    struct=NiagaraExt_VariableValue_DataInterface
    value {dataInterface=&self:NiagaraDataInterfaceSkeletalMesh_0}
   }
  }
 ]
}
