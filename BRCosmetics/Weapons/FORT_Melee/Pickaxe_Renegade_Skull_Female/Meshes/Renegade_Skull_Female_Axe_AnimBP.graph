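# Animation Blueprint for the Renegade Skull Female pickaxe that makes four dyn_spine segments float and bob procedurally.
# Each update, game-time sine and cosine waves at differing frequencies set per-segment offset vectors and rotators.
# After Copy Pose From Mesh, chained Modify Bone nodes add translation to both bones per segment, rotation only to the first.
# Zero-gravity Anim Dynamics springs, simulated relative to b_AR_Root, then add motion lag on each second bone.
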AnimBlueprint /BRCosmetics/Weapons/FORT_Melee/Pickaxe_Renegade_Skull_Female/Meshes/Renegade_Skull_Female_Axe_AnimBP
target_skeleton=&/Renegade_Skull_Female_Axe_Skeleton
graphs [
 AnimGraph {nodes [
   n1 = Get Spine_01_Float() -> (Spine_01_Float:Vector)
   n2 = Get Floating_Rotator1() -> (Floating_Rotator1:Rotator)
   n3 = "Transform (Modify) Bone - Bone: dyn_spine_1_1"(ComponentPose=Local To Component(LocalPose=Copy Pose From Mesh()), Alpha=1, Translation=@n1, Rotation=@n2, Scale="1,1,1") {rotationMode=BMM_Additive, translationMode=BMM_Additive, alphaBoolBlend {blendOption=Linear}, boneToModify {boneName=dyn_spine_1_1}}
   n4 = Get Spine_02_Float() -> (Spine_02_Float:Vector)
   n5 = Get Floating_Rotator2() -> (Floating_Rotator2:Rotator)
   n6 = "Transform (Modify) Bone - Bone: dyn_spine_2_1"(
    ComponentPose="Transform (Modify) Bone - Bone: dyn_spine_1_2"(ComponentPose=@n3, Alpha=1, Translation=@n1, Rotation=@n2, Scale="1,1,1") {translationMode=BMM_Additive, alphaBoolBlend {blendOption=Linear}, boneToModify {boneName=dyn_spine_1_2}}
    Alpha=1
    Translation=@n4
    Rotation=@n5
    Scale="1,1,1"
   ) {rotationMode=BMM_Additive, translationMode=BMM_Additive, alphaBoolBlend {blendOption=Linear}, boneToModify {boneName=dyn_spine_2_1}}
   n7 = Get Spine_03_Float() -> (Spine_03_Float:Vector)
   n8 = Get Floating_Rotator3() -> (Floating_Rotator3:Rotator)
   n9 = "Transform (Modify) Bone - Bone: dyn_spine_3_1"(
    ComponentPose="Transform (Modify) Bone - Bone: dyn_spine_2_2"(ComponentPose=@n6, Alpha=1, Translation=@n4, Rotation=@n5, Scale="1,1,1") {translationMode=BMM_Additive, alphaBoolBlend {blendOption=Linear}, boneToModify {boneName=dyn_spine_2_2}}
    Alpha=1
    Translation=@n7
    Rotation=@n8
    Scale="1,1,1"
   ) {rotationMode=BMM_Additive, translationMode=BMM_Additive, alphaBoolBlend {blendOption=Linear}, boneToModify {boneName=dyn_spine_3_1}}
   n10 = Get Spine_04_Float() -> (Spine_04_Float:Vector)
   n11 = Get Floating_Rotator4() -> (Floating_Rotator4:Rotator)
   n12 = "Transform (Modify) Bone - Bone: dyn_spine_4_1"(
    ComponentPose="Transform (Modify) Bone - Bone: dyn_spine_3_2"(ComponentPose=@n9, Alpha=1, Translation=@n7, Rotation=@n8, Scale="1,1,1") {translationMode=BMM_Additive, alphaBoolBlend {blendOption=Linear}, boneToModify {boneName=dyn_spine_3_2}}
    Alpha=1
    Translation=@n10
    Rotation=@n11
    Scale="1,1,1"
   ) {rotationMode=BMM_Additive, translationMode=BMM_Additive, alphaBoolBlend {blendOption=Linear}, boneToModify {boneName=dyn_spine_4_1}}
   n13 = Anim Dynamics - Bone: dyn_spine_1_2(
    ComponentPose="Transform (Modify) Bone - Bone: dyn_spine_4_2"(ComponentPose=@n12, Alpha=1, Translation=@n10, Rotation=@n11, Scale="1,1,1") {translationMode=BMM_Additive, alphaBoolBlend {blendOption=Linear}, boneToModify {boneName=dyn_spine_4_2}}
    Alpha=1
    ExternalForce="0,0,0"
   ) {
    angularSpringConstant=80
    bAngularSpring=true
    bGravityOverrideInSimSpace=true
    bLinearSpring=true
    bOverrideLinearDamping=true
    gravityScale=0
    linearDampingOverride=0.95
    linearSpringConstant=20
    simulationSpace=RootRelative
    alphaBoolBlend {blendOption=Linear}
    boundBone {boneName=dyn_spine_1_2}
    componentAppliedLinearAccClamp=<500 500 500>
    componentLinearAccScale=<0.1 0.1 0.05>
    componentLinearVelScale=<0.2 0.2 0.05>
    physicsBodyDefinitions [{
      sphereCollisionRadius=0
      boundBone {boneName=dyn_spine_1_2}
      boxExtents=<4 4 2>
      constraintSetup {angularLimitsMax=<0 0 0>, angularLimitsMin=<0 0 0>, angularTarget=<0 0 0>, linearAxesMax=<0.25 0.25 2.5>, linearAxesMin=<-0.25 -0.25 -1>}
      localJointOffset=<-0.05 -0.05 0>
     }]
    relativeSpaceBone {boneName=b_AR_Root}
   }
   n14 = Anim Dynamics - Bone: dyn_spine_2_2(ComponentPose=@n13, Alpha=1, ExternalForce="0,0,0") {
    angularSpringConstant=60
    bAngularSpring=true
    bGravityOverrideInSimSpace=true
    bLinearSpring=true
    bOverrideLinearDamping=true
    gravityScale=0
    linearDampingOverride=0.95
    linearSpringConstant=30
    simulationSpace=RootRelative
    alphaBoolBlend {blendOption=Linear}
    boundBone {boneName=dyn_spine_2_2}
    componentAppliedLinearAccClamp=<6000 6000 6000>
    componentLinearAccScale=<0.1 0.1 0.05>
    componentLinearVelScale=<0.2 0.2 0.05>
    physicsBodyDefinitions [{
      sphereCollisionRadius=0
      boundBone {boneName=dyn_spine_2_2}
      boxExtents=<4 4 2>
      constraintSetup {angularLimitsMax=<0 0 0>, angularLimitsMin=<0 0 0>, angularTarget=<0 0 0>, linearAxesMax=<0.5 0.5 5>, linearAxesMin=<-0.5 -0.5 -1>}
      localJointOffset=<-0.15 -0.05 0>
     }]
    relativeSpaceBone {boneName=b_AR_Root}
   }
   n15 = Anim Dynamics - Bone: dyn_spine_3_2(ComponentPose=@n14, Alpha=1, ExternalForce="0,0,0") {
    angularSpringConstant=90
    bAngularSpring=true
    bGravityOverrideInSimSpace=true
    bLinearSpring=true
    bOverrideLinearDamping=true
    gravityScale=0
    linearDampingOverride=0.95
    linearSpringConstant=40
    simulationSpace=RootRelative
    alphaBoolBlend {blendOption=Linear}
    boundBone {boneName=dyn_spine_3_2}
    componentAppliedLinearAccClamp=<6000 6000 6000>
    componentLinearAccScale=<0.1 0.1 0.05>
    componentLinearVelScale=<0.2 0.2 0.05>
    physicsBodyDefinitions [{
      sphereCollisionRadius=0
      boundBone {boneName=dyn_spine_3_2}
      boxExtents=<4 4 2>
      constraintSetup {angularLimitsMax=<0 0 0>, angularLimitsMin=<0 0 0>, angularTarget=<0 0 0>, linearAxesMax=<1 1 8>, linearAxesMin=<-1 -1 -1>}
      localJointOffset=<0 -0.2 0>
     }]
    relativeSpaceBone {boneName=b_AR_Root}
   }
   n16 = Anim Dynamics - Bone: dyn_spine_4_2(ComponentPose=@n15, Alpha=1, ExternalForce="0,0,0") {
    angularDampingOverride=0.85
    angularSpringConstant=80
    bAngularSpring=true
    bGravityOverrideInSimSpace=true
    bLinearSpring=true
    bOverrideAngularDamping=true
    bOverrideLinearDamping=true
    gravityScale=0
    linearDampingOverride=0.95
    linearSpringConstant=37.5
    simulationSpace=RootRelative
    alphaBoolBlend {blendOption=Linear}
    boundBone {boneName=dyn_spine_4_2}
    componentAppliedLinearAccClamp=<6000 6000 6000>
    componentLinearAccScale=<0.1 0.1 0.1>
    componentLinearVelScale=<0.2 0.2 0.2>
    physicsBodyDefinitions [{
      sphereCollisionRadius=0
      boundBone {boneName=dyn_spine_4_2}
      boxExtents=<5 5 2.5>
      constraintSetup {angularLimitsMax=<0 0 0>, angularLimitsMin=<0 0 0>, angularTarget=<0 0 0>, linearAxesMax=<1.25 1.25 10>, linearAxesMin=<-1.25 0 -1>}
      localJointOffset=<0 0 0>
     }]
    relativeSpaceBone {boneName=b_AR_Root}
   }
   Output Pose(Result=Component To Local(ComponentPose=@n16)) {layerGroup=None}
  ]}
 EventGraph {nodes [
   Event BlueprintUpdateAnimation()
   n1 = GetGameTimeInSeconds()
   n2 = float * float(A=float * float(A=0.95, B=1), B=@n1)
   n3 = float * float(A=float * float(A=0.95, B=1), B=@n1)
   n4 = float * float(A=float * float(A=1, B=0.5), B=@n1)
   n5 = MakeVector(
    X=float * float(A="Sin (Radians)"(A=@n2), B=0.3)
    Y=float * float(A="Sin (Radians)"(A=@n3), B=0.3)
    Z=float * float(A="Sin (Radians)"(A=@n4), B=0.1)
   )
   > Set Spine_01_Float(Spine_01_Float:Vector=@n5) -> (Output_Get:Vector)
   n6 = BreakVector(InVec=vector * float(A=@n5, B=15))
   > Set Floating_Rotator1(Floating_Rotator1:Rotator=MakeRotator(Roll=@n6, Pitch=@n6.Y, Yaw=@n6.Z)) -> (Output_Get:Rotator)
   n7 = float * float(A=float * float(A=1, B=1.1), B=@n1)
   n8 = float * float(A=float * float(A=1, B=0.95), B=@n1)
   n9 = float * float(A=float * float(A=1, B=0.55), B=@n1)
   n10 = MakeVector(
    X=float * float(A="Cos (Radians)"!(A:double=@n7) -> (ReturnValue:double), B=0.3)
    Y=float * float(A="Cos (Radians)"(A=@n8), B=0.3)
    Z=float * float(A="Cos (Radians)"(A=@n9), B=0.1)
   )
   > Set Spine_02_Float(Spine_02_Float:Vector=@n10) -> (Output_Get:Vector)
   n11 = BreakVector(InVec=vector * float(A=@n10, B=15))
   > Set Floating_Rotator2(Floating_Rotator2:Rotator=MakeRotator(Roll=@n11, Pitch=@n11.Y, Yaw=@n11.Z)) -> (Output_Get:Rotator)
   n12 = float * float(A=float * float(A=1.05, B=0.95), B=@n1)
   n13 = float * float(A=float * float(A=1.05, B=1.1), B=@n1)
   n14 = float * float(A=float * float(A=1.05, B=0.45), B=@n1)
   n15 = MakeVector(
    X=float * float(A="Sin (Radians)"(A=@n12), B=0.3)
    Y=float * float(A="Sin (Radians)"(A=@n13), B=0.3)
    Z=float * float(A="Sin (Radians)"(A=@n14), B=0.1)
   )
   > Set Spine_03_Float(Spine_03_Float:Vector=@n15) -> (Output_Get:Vector)
   n16 = BreakVector(InVec=vector * float(A=@n15, B=15))
   > Set Floating_Rotator3(Floating_Rotator3:Rotator=MakeRotator(Roll=@n16, Pitch=@n16.Y, Yaw=@n16.Z)) -> (Output_Get:Rotator)
   n17 = float * float(A=float * float(A=1.1, B=1.15), B=@n1)
   n18 = float * float(A=float * float(A=0.95, B=1.1), B=@n1)
   n19 = float * float(A=float * float(A=1.15, B=0.95), B=@n1)
   n20 = MakeVector(
    X=float * float(A="Sin (Radians)"(A=@n17), B=0.3)
    Y=float * float(A="Sin (Radians)"(A=@n19), B=0.3)
    Z=float * float(A="Sin (Radians)"(A=@n18), B=0.15)
   )
   > Set Spine_04_Float(Spine_04_Float:Vector=vector * float(A=@n20, B=0.5)) -> (Output_Get:Vector)
   n21 = BreakVector(InVec=vector * float(A=@n20, B=2.5))
   > Set Floating_Rotator4(Floating_Rotator4:Rotator=MakeRotator(Roll=@n21, Pitch=@n21.Y, Yaw=@n21.Z)) -> (Output_Get:Rotator)
  ]}
]
variables [
 Floating_Rotator1 {type=Rotator}
 Floating_Rotator2 {type=Rotator}
 Floating_Rotator3 {type=Rotator}
 Floating_Rotator4 {type=Rotator}
 Spine_01_Float {type=Vector}
 Spine_02_Float {type=Vector}
 Spine_03_Float {type=Vector}
 Spine_04_Float {type=Vector}
]
